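# Sound cue for the foley played when searching a couch loot container completes, routed to the Fort_Loot_Collect sound class.
# A Random node picks one of three equally weighted couch search waves and feeds an Enveloper whose volume curve rises quickly, then fades to silence, keeping it short.
# Prop-search attenuation spatializes it, and a concurrency override allows one instance at a time, blocking new plays while one is active.
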SoundCue /Game/Sounds/Foley_Loot/Containers/Sofa/Prop_Couch_SearchComplete_Cue
root=Enveloper_7
nodes [
 {
  id=Enveloper_7
  children [Random_0]
  properties {pitchCurve {editorCurveData {keys [{time=0}]}}, volumeCurve {editorCurveData {keys [<at 0 0>,<at 0.1176 0.8234>,<at 0.64 0>]}}}
 }
 {id=Random_0, children [WavePlayer_3,WavePlayer_4,WavePlayer_5], properties {weights [1,1,1]}}
 {id=WavePlayer_3, properties {soundWaveAssetPtr=&/Prop_Couch_Search_04}}
 {id=WavePlayer_4, properties {soundWaveAssetPtr=&/Prop_Couch_Search_05}}
 {id=WavePlayer_5, properties {soundWaveAssetPtr=&/Prop_Couch_Search_06}}
]
properties {
 attenuationSettings=/Game/Sounds/Attenuation/Fort_Prop_Search_Att
 bOverrideConcurrency=true
 maxDistance=1300
 soundClassObject=&Fort_Loot_Collect
 volumeMultiplier=0.45
 concurrencyOverrides {maxCount=1, resolutionRule=PreventNew}
}
