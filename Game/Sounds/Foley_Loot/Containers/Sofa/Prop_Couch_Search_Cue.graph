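# Sound cue for searching a couch loot container, layering couch rustle with gear rummaging.
# A mixer plays one of three random couch search waves at reduced level alongside one of three random generic gear search waves at boosted level, the latter delayed by one second.
# It uses the loot collect sound class, prop search attenuation and a concurrency override that prevents new instances at the limit.
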SoundCue /Game/Sounds/Foley_Loot/Containers/Sofa/Prop_Couch_Search_Cue
root=Mixer_26
nodes [
 {id=Delay_35, children [Random_25], properties {delayMax=1, delayMin=1}}
 {id=Mixer_26, children [Random_9,Delay_35], properties {inputVolume [0.4,1.2]}}
 {id=Random_25, children [WavePlayer_129, WavePlayer_130, WavePlayer_132], properties {weights [1,1,1]}}
 {id=Random_9, children [WavePlayer_3,WavePlayer_4,WavePlayer_5], properties {weights [1,1,1]}}
 {id=WavePlayer_129, properties {soundWaveAssetPtr=/Game/Sounds/Foley_Loot/Containers/Misc/Prop_Gear_Search_05}}
 {id=WavePlayer_130, properties {soundWaveAssetPtr=/Game/Sounds/Foley_Loot/Containers/Misc/Prop_Gear_Search_06}}
 {id=WavePlayer_132, properties {soundWaveAssetPtr=/Game/Sounds/Foley_Loot/Containers/Misc/Prop_Gear_Search_01}}
 {id=WavePlayer_3, properties {soundWaveAssetPtr=&/Prop_Couch_Search_04}}
 {id=WavePlayer_4, properties {soundWaveAssetPtr=&/Prop_Couch_Search_05}}
 {id=WavePlayer_5, properties {soundWaveAssetPtr=&/Prop_Couch_Search_06}}
]
properties {
 attenuationSettings=/Game/Sounds/Attenuation/Fort_Prop_Search_Att
 bOverrideConcurrency=true
 maxDistance=1300
 soundClassObject=&Fort_Loot_Collect
 volumeMultiplier=0.7
 concurrencyOverrides {resolutionRule=PreventNew}
}
